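# Physics asset for the HollyDessert backpack skeleton, giving capsule bodies to its upper spine, neck, head and hair bones.
# Constraints chain the spine through the neck to the head and on to a hair master bone, all with free swing and twist.
# Below the master, a five-segment center hair chain uses tight angular limits, with left and right strands branching from its fourth segment under the same limits.
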PhysicsAsset /BRCosmetics/Accessories/FORT_Backpacks/Backpack_F_MED_HollyDessert/Meshes/F_MED_HollyDessert_Pack_Physics
bodies {
 pack_hair_WLD_c_01 [{shapeType=Capsule, length=3, radius=1, shapeName=pack_hair_WLD_c_01_capsule, center=<1.3554 0 0>}]
 pack_hair_WLD_c_02 [{shapeType=Capsule, length=3, radius=1.2, shapeName=pack_hair_WLD_c_02_capsule, center=<1.1346 0 0>}]
 pack_hair_WLD_c_03 [{shapeType=Capsule, length=3, radius=1.2, shapeName=pack_hair_WLD_c_03_capsule, center=<1.1346 0 0>}]
 pack_hair_WLD_c_04 [{shapeType=Capsule, length=3, radius=1.2, shapeName=pack_hair_WLD_c_04_capsule, center=<1.1346 0 0>}]
 pack_hair_WLD_c_05 [{shapeType=Capsule, length=3, radius=1.3, shapeName=pack_hair_WLD_c_05_capsule, center=<1.1346 0 0>}]
 pack_hair_WLD_c_MSTR [{shapeType=Capsule, length=0.505, radius=0.505, shapeName=pack_hair_WLD_c_MSTR_capsule}]
 pack_hair_WLD_l_01 [{shapeType=Capsule, length=3, radius=1, shapeName=pack_hair_WLD_l_01_capsule, center=<2.6021 0 0>}]
 pack_hair_WLD_r_01 [{shapeType=Capsule, length=3, radius=1, shapeName=pack_hair_WLD_r_01_capsule, center=<2.6021 0 0>}]
 pack_head_WLD_01 [{shapeType=Capsule, length=0.5, radius=2.7, shapeName=pack_head_WLD_01_capsule, center=<1.1633 1.0815 0>, rotation=<rot 90 0 0>}]
 pack_neck_WLD_01 [{shapeType=Capsule, length=1, radius=1.2, shapeName=pack_neck_WLD_01_capsule, rotation=<rot 90 0 0>}]
 pack_neck_WLD_02 [{shapeType=Capsule, length=1, radius=1.2, shapeName=pack_neck_WLD_02_capsule, rotation=<rot 90 0 0>}]
 pack_spine_WLD_03 [{shapeType=Capsule, length=3.505, radius=2.505, shapeName=pack_spine_WLD_03_capsule, center=<0 0.7232 0>, rotation=<rot 0 -90 0>}]
 pack_spine_WLD_04 [{shapeType=Capsule, length=3.505, radius=2.505, shapeName=pack_spine_WLD_04_capsule, center=<0 0.7232 0>, rotation=<rot 0 -90 0>}]
 pack_spine_WLD_05 [{shapeType=Capsule, length=3.505, radius=2.505, shapeName=pack_spine_WLD_05_capsule, center=<0 0.7232 0>, rotation=<rot 0 -90 0>}]
}
constraints [
 {bone1Name=pack_hair_WLD_c_01, bone2Name=pack_hair_WLD_c_MSTR, swing1LimitDegrees=15, swing2LimitDegrees=10, twistLimitDegrees=10}
 {bone1Name=pack_hair_WLD_c_02, bone2Name=pack_hair_WLD_c_01, swing1LimitDegrees=15, swing2LimitDegrees=10, twistLimitDegrees=10}
 {bone1Name=pack_hair_WLD_c_03, bone2Name=pack_hair_WLD_c_02, swing1LimitDegrees=15, swing2LimitDegrees=10, twistLimitDegrees=10}
 {bone1Name=pack_hair_WLD_c_04, bone2Name=pack_hair_WLD_c_03, swing1LimitDegrees=15, swing2LimitDegrees=10, twistLimitDegrees=10}
 {bone1Name=pack_hair_WLD_c_05, bone2Name=pack_hair_WLD_c_04, swing1LimitDegrees=15, swing2LimitDegrees=10, twistLimitDegrees=10}
 {bone1Name=pack_hair_WLD_l_01, bone2Name=pack_hair_WLD_c_04, swing1LimitDegrees=15, swing2LimitDegrees=10, twistLimitDegrees=10}
 {bone1Name=pack_hair_WLD_r_01, bone2Name=pack_hair_WLD_c_04, swing1LimitDegrees=15, swing2LimitDegrees=10, twistLimitDegrees=10}
 {
  bone1Name=pack_hair_WLD_c_MSTR
  bone2Name=pack_head_WLD_01
  swing1LimitDegrees=45
  swing1Motion=Free
  swing2LimitDegrees=45
  swing2Motion=Free
  twistLimitDegrees=45
  twistMotion=Free
 }
 {
  bone1Name=pack_spine_WLD_04
  bone2Name=pack_spine_WLD_03
  swing1LimitDegrees=45
  swing1Motion=Free
  swing2LimitDegrees=45
  swing2Motion=Free
  twistLimitDegrees=45
  twistMotion=Free
 }
 {
  bone1Name=pack_spine_WLD_05
  bone2Name=pack_spine_WLD_04
  swing1LimitDegrees=45
  swing1Motion=Free
  swing2LimitDegrees=45
  swing2Motion=Free
  twistLimitDegrees=45
  twistMotion=Free
 }
 {
  bone1Name=pack_neck_WLD_01
  bone2Name=pack_spine_WLD_05
  swing1LimitDegrees=45
  swing1Motion=Free
  swing2LimitDegrees=45
  swing2Motion=Free
  twistLimitDegrees=45
  twistMotion=Free
 }
 {
  bone1Name=pack_neck_WLD_02
  bone2Name=pack_neck_WLD_01
  swing1LimitDegrees=45
  swing1Motion=Free
  swing2LimitDegrees=45
  swing2Motion=Free
  twistLimitDegrees=45
  twistMotion=Free
 }
 {
  bone1Name=pack_head_WLD_01
  bone2Name=pack_neck_WLD_02
  swing1LimitDegrees=45
  swing1Motion=Free
  swing2LimitDegrees=45
  swing2Motion=Free
  twistLimitDegrees=45
  twistMotion=Free
 }
]
