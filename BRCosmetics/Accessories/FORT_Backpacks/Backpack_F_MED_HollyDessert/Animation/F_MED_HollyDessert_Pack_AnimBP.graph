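# Animation Blueprint for the HollyDessert backpack, with idle and reactive animations in game and in the frontend.
# A bool blends between a frontend state machine and an in-game one driven by damage and enemy-elimination anim input events.
# When the backpack's part modifier enables reactivity, delays replay the elimination animation periodically in the frontend.
# Position constraints pin WLD spine, neck and head bones to their counterparts before a root-space RigidBody simulation.
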AnimBlueprint /BRCosmetics/Accessories/FORT_Backpacks/Backpack_F_MED_HollyDessert/Animation/F_MED_HollyDessert_Pack_AnimBP
target_skeleton=/BRCosmetics/Accessories/FORT_Backpacks/Backpack_F_MED_HollyDessert/Meshes/F_MED_HollyDessert_Pack_Skeleton
graphs [
 AnimGraph {nodes [
   AnimGraphNode_StateMachine_1 = Backpack States() {is=title AnimGraphNode_StateMachine}
   AnimGraphNode_StateMachine_2 = FE Idle + Reactive() {is=title AnimGraphNode_StateMachine}
   n1 = Get bShouldDisplayFrontendReactivity() -> (bShouldDisplayFrontendReactivity:bool)
   n2 = "Constraint - pack_spine_WLD_03 (P)"(
    ComponentPose=Local To Component(LocalPose=Blend Poses by bool(BlendPose_0=@AnimGraphNode_StateMachine_2, BlendPose_1=@AnimGraphNode_StateMachine_1, bActiveValue=@n1) {blendType=Linear})
    Alpha=1
    ConstraintWeights_0=1
   ) {
    alphaBoolBlend {blendOption=Linear}
    boneToModify {boneName=pack_spine_WLD_03}
    constraintSetup [{targetBone {boneName=pack_spine_03}}]
    constraintWeights [1]
   }
   n3 = "Constraint - pack_spine_WLD_04 (P)"(ComponentPose=@n2, Alpha=1, ConstraintWeights_0=1) {
    alphaBoolBlend {blendOption=Linear}
    boneToModify {boneName=pack_spine_WLD_04}
    constraintSetup [{targetBone {boneName=pack_spine_04}}]
    constraintWeights [1]
   }
   n4 = "Constraint - pack_spine_WLD_05 (P)"(ComponentPose=@n3, Alpha=1, ConstraintWeights_0=1) {
    alphaBoolBlend {blendOption=Linear}
    boneToModify {boneName=pack_spine_WLD_05}
    constraintSetup [{targetBone {boneName=pack_spine_05}}]
    constraintWeights [1]
   }
   n5 = "Constraint - pack_neck_WLD_01 (P)"(ComponentPose=@n4, Alpha=1, ConstraintWeights_0=1) {
    alphaBoolBlend {blendOption=Linear}
    boneToModify {boneName=pack_neck_WLD_01}
    constraintSetup [{targetBone {boneName=pack_neck_01}}]
    constraintWeights [1]
   }
   n6 = "Constraint - pack_neck_WLD_02 (P)"(ComponentPose=@n5, Alpha=1, ConstraintWeights_0=1) {
    alphaBoolBlend {blendOption=Linear}
    boneToModify {boneName=pack_neck_WLD_02}
    constraintSetup [{targetBone {boneName=pack_neck_02}}]
    constraintWeights [1]
   }
   n7 = "Constraint - pack_head_WLD_01 (P)"(ComponentPose=@n6, Alpha=1, ConstraintWeights_0=1) {
    alphaBoolBlend {blendOption=Linear}
    boneToModify {boneName=pack_head_WLD_01}
    constraintSetup [{targetBone {boneName=pack_head}}]
    constraintWeights [1]
   }
   n8 = RigidBody(ComponentPose=@n7) {
    bClampLinearTranslationLimitToRefPose=true
    bOverrideWorldGravity=true
    overridePhysicsAsset=/BRCosmetics/Accessories/FORT_Backpacks/Backpack_F_MED_HollyDessert/Meshes/F_MED_HollyDessert_Pack_Physics
    simulationSpace=BaseBoneSpace
    alphaBoolBlend {blendOption=Linear}
    baseBoneRef {boneName=root}
    componentAppliedLinearAccClamp=<2000 2000 2000>
    componentLinearAccScale=<1 0.3 0.3>
    componentLinearVelScale=<1 0.3 0.3>
    overrideWorldGravity {z=-980}
   }
   Output Pose(Result=Component To Local(ComponentPose=@n8)) {layerGroup=None}
  ]}
 Backpack States {
  entry=Idle
  path=AnimGraph.AnimGraphNode_StateMachine_1.Backpack States
  ids=id
  nodes [
   AnimStateNode_0 = Idle() {stateEntered {guid=#1}}
   AnimStateNode_1 = Took Damage() {is=title AnimStateNode}
   AnimStateNode_4 = Elimination() {is=title AnimStateNode}
   AnimStateTransitionNode_11 = AnimStateTransitionNode()
   AnimStateTransitionNode_12 = AnimStateTransitionNode()
   AnimStateTransitionNode_13 = AnimStateTransitionNode() {bAutomaticRuleBasedOnSequencePlayerInState=true}
   AnimStateTransitionNode_5 = AnimStateTransitionNode() {bAutomaticRuleBasedOnSequencePlayerInState=true}
  ]
  links [
   {
    dst=Idle
    rule=AnimGraph.AnimGraphNode_StateMachine_1.Backpack States.AnimStateTransitionNode_5.Transition
    src=Elimination
    via=AnimStateTransitionNode_5
   }
   {
    dst=Elimination
    rule=AnimGraph.AnimGraphNode_StateMachine_1.Backpack States.AnimStateTransitionNode_11.Transition
    src=Idle
    via=AnimStateTransitionNode_11
   }
   {
    dst=Took Damage
    rule=AnimGraph.AnimGraphNode_StateMachine_1.Backpack States.AnimStateTransitionNode_12.Transition
    src=Idle
    via=AnimStateTransitionNode_12
   }
   {
    dst=Idle
    rule=AnimGraph.AnimGraphNode_StateMachine_1.Backpack States.AnimStateTransitionNode_13.Transition
    src=Took Damage
    via=AnimStateTransitionNode_13
   }
  ]
 }
 Idle {
  path=AnimGraph.AnimGraphNode_StateMachine_1.Backpack States.AnimStateNode_0.Idle
  nodes [Output Animation Pose(Result=Sequence Player 'HollyDessert_idle_PACK'() {sequence=&/HollyDessert_idle_PACK})]
 }
 Took Damage {
  path=AnimGraph.AnimGraphNode_StateMachine_1.Backpack States.AnimStateNode_1.Took Damage
  nodes [
   n1 = Sequence Player 'HollyDessert_Damage_PACK'() {bLoopAnimation=false, sequence=&/HollyDessert_Damage_PACK}
   Output Animation Pose(Result=@n1)
  ]
 }
 Elimination {
  path=AnimGraph.AnimGraphNode_StateMachine_1.Backpack States.AnimStateNode_4.Elimination
  nodes [
   n1 = Sequence Player 'HollyDessert_Elimination_PACK'() {bLoopAnimation=false, sequence=&/HollyDessert_Elimination_PACK}
   Output Animation Pose(Result=@n1)
  ]
 }
 Transition {
  path=AnimGraph.AnimGraphNode_StateMachine_1.Backpack States.AnimStateTransitionNode_11.Transition
  nodes [Result(bCanEnterTransition=Get bEnemyPlayerEliminated() -> (bEnemyPlayerEliminated:bool))]
 }
 Transition {
  path=AnimGraph.AnimGraphNode_StateMachine_1.Backpack States.AnimStateTransitionNode_12.Transition
  nodes [Result(bCanEnterTransition=Get TookDamage() -> (TookDamage:bool))]
 }
 Transition {path=AnimGraph.AnimGraphNode_StateMachine_1.Backpack States.AnimStateTransitionNode_13.Transition, nodes [Result()]}
 Transition {path=AnimGraph.AnimGraphNode_StateMachine_1.Backpack States.AnimStateTransitionNode_5.Transition, nodes [Result()]}
 FE Idle + Reactive {
  entry=FE_Idle
  path=AnimGraph.AnimGraphNode_StateMachine_2.FE Idle + Reactive
  ids=id
  nodes [
   AnimStateNode_0 = FE_Idle() {is=title AnimStateNode}
   AnimStateNode_1 = FE_Reactive() {is=title AnimStateNode, stateEntered {guid=#2, notifyName=Entered FE_Reactive_Anim}}
   AnimStateTransitionNode_0 = AnimStateTransitionNode()
   AnimStateTransitionNode_1 = AnimStateTransitionNode() {bAutomaticRuleBasedOnSequencePlayerInState=true}
  ]
  links [
   {
    dst=FE_Reactive
    rule=AnimGraph.AnimGraphNode_StateMachine_2.FE Idle + Reactive.AnimStateTransitionNode_0.Transition
    src=FE_Idle
    via=AnimStateTransitionNode_0
   }
   {
    dst=FE_Idle
    rule=AnimGraph.AnimGraphNode_StateMachine_2.FE Idle + Reactive.AnimStateTransitionNode_1.Transition
    src=FE_Reactive
    via=AnimStateTransitionNode_1
   }
  ]
 }
 FE_Idle {
  path=AnimGraph.AnimGraphNode_StateMachine_2.FE Idle + Reactive.AnimStateNode_0.FE_Idle
  nodes [Output Animation Pose(Result=Sequence Player 'HollyDessert_idle_PACK'() {sequence=&HollyDessert_idle_PACK})]
 }
 FE_Reactive {
  path=AnimGraph.AnimGraphNode_StateMachine_2.FE Idle + Reactive.AnimStateNode_1.FE_Reactive
  nodes [
   n1 = Sequence Player 'HollyDessert_Elimination_PACK'() {bLoopAnimation=false, sequence=&HollyDessert_Elimination_PACK}
   Output Animation Pose(Result=@n1)
  ]
 }
 Transition {
  path=AnimGraph.AnimGraphNode_StateMachine_2.FE Idle + Reactive.AnimStateTransitionNode_0.Transition
  nodes [Result(bCanEnterTransition=Get Play Reactive FE Anim() -> (Play Reactive FE Anim:bool))]
 }
 Transition {path=AnimGraph.AnimGraphNode_StateMachine_2.FE Idle + Reactive.AnimStateTransitionNode_1.Transition, nodes [Result()]}
 EventGraph {nodes [
   n1 = Cast To FortPlayerPawn(Object=GetOwningActor())
   n2 = GetHealth!(self:Fort Pawn Object Reference=@n1) -> (ReturnValue:float)
   > Set Current Health(Current Health:double=@n2) -> (Output_Get:double)
   > n3 = Branch(Condition="Not Equal (Float)"(A=@n2, B=Get Previous Health() -> (Previous Health:double)))
   > Set TookDamage(TookDamage:bool=true) -> (Output_Get:bool)
   > Set Previous Health(Previous Health:double=Get Current Health() -> (Current Health:double)) -> (Output_Get:double)
   Set TookDamage(execute=@n3.else, TookDamage=false)
   Event OnCharacterCustomizationLoaded()
   > n4 = Cast To FortPlayerPawn(Object=TryGetFortPawn())
   > n5 = Cast To B_Athena_PartModifier_Backpack_HollyDessert!(
    execute
    Object=GetModifierForPartType!(self:Fort Player Pawn Object Reference=@n4, PartType:EFortCustomPartType Enum=Backpack) -> (ReturnValue:Custom Character Part Modifier Object Reference)
   ) -> (then, CastFailed, AsB Athena Part Modifier Backpack Holly Dessert:B Athena Part Modifier Backpack Holly Dessert Object Reference) {is=title K2Node_DynamicCast}
   > n6 = Set Reactivity Enabled(Reactivity Enabled:bool=Get Reactivity_EnabledMod!(self:B Athena Part Modifier Backpack Holly Dessert Object Reference=@n5) -> (Reactivity_EnabledMod:bool)) -> (Output_Get:bool)
   > n7 = Branch(Condition=@n6)
   > Do Once(Reset=@n9.then)
   > n8 = Delay(Duration=4)
   Set Play Reactive FE Anim(execute=[@n8.then, @n10.then], Play Reactive FE Anim:bool=true) -> (Output_Get:bool)
   n9 = Set Play Reactive FE Anim(execute=@n7.else, Play Reactive FE Anim=false)
   AnimNotify_Entered FE_Reactive_Anim() {is=title K2Node_Event}
   > Set Play Reactive FE Anim(Play Reactive FE Anim=false)
   > n10 = Delay(Duration=8)
   Event BlueprintPostEvaluateAnimation() {is=title K2Node_Event}
   > Set bEnemyPlayerEliminated(bEnemyPlayerEliminated:bool=false) -> (Output_Get:bool)
   n11 = Event BlueprintProcessAnimInputEvent()
   > n12 = Sequence()
   > Branch(Condition=IsNamed(self=@n11.AnimInputEvent, InEventName=aie.EnemyPlayerEliminated))
   > Set bEnemyPlayerEliminated(bEnemyPlayerEliminated=true)
   Branch(
    execute=@n12.then_1
    Condition=OR Boolean(
     A=IsNamed(self=@n11.AnimInputEvent, InEventName=aie.PlayerVitality.HealthDecreased)
     B=IsNamed(self=@n11.AnimInputEvent, InEventName=Noneaie.PlayerVitality.ShieldsDecreased)
    )
   )
   > Set TookDamage(TookDamage=true)
   > RetriggerableDelay(Duration=0.2)
   > Set TookDamage(TookDamage=false)
   TryGetPawnOwner()
   Event BlueprintUpdateAnimation()
  ]}
]
variables [
 Active_Kill {type=boolean}
 As B Athena Part Modifier Backpack Space Feline {type=B_Athena_PartModifier_Backpack_SpaceFeline_C, default=null}
 bEnemyPlayerEliminated {type=boolean}
 Current Health {type=number}
 "FInterp(Alpha, (10: 10))" {type=number}
 "FInterp(Alpha, (10: 10))_0" {type=number}
 KillCount {type=number}
 Play Reactive FE Anim {type=boolean}
 Previous Health {type=number}
 Reactivity Enabled {type=boolean}
 Skel_Mesh {type=SkeletalMeshComponent}
 TookDamage {type=boolean}
]
