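# Parent material for the HollyDessert backpack, layering an animated liquid effect onto the 2023 anime character shader.
# A UV sphere mask defines the liquid area, and local-position gradients locate a character in it for darkening and radial coordinates.
# Noise-distorted, scrolling Voronoi normals create ripples, while adjusted STT channels add liquid specular.
# Flicker-driven rotation about a pivot wobbles the liquid through World Position Offset.
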Material /BRCosmetics/Accessories/FORT_Backpacks/Backpack_F_MED_HollyDessert/Materials/M_HollyDessert_Backpack_Parent
graph {
 properties {
  bAutomaticallySetUsageInEditor=false
  bUseMaterialAttributes=true
  bUsedWithMeshDeformer=true
  bUsedWithMorphTargets=true
  bUsedWithSkeletalMesh=true
  bUsedWithStaticLighting=true
  pixelDepthOffsetMode=PDOM_Legacy
  previewMesh=/BRCosmetics/Accessories/FORT_Backpacks/Backpack_F_MED_HollyDessert/Meshes/F_MED_HollyDessert_Pack
  shadingModel=MSM_Cloth
  usedShadingModels=Cloth
 }
 nodes [
  n1 = NamedRerouteUsage()
  n2 = TextureSample(UVs=@n1) {samplerType=SAMPLERTYPE_LinearColor, texture=&HighContrastNoise}
  Multiply(A=@n2.B)
  n3 = SphereMask(A=TextureCoordinate(), B=Constant2Vector() {g=0.625, r=0.4625}) {attenuationRadius=0.0475, hardnessPercent=0}
  NamedRerouteDeclaration(None=Ceil(None=NamedRerouteDeclaration(None=@n3) {name=LiquidGradient, nodeColor=<rgba 0 1 0.0859 1>})) {name=LiquidMask, nodeColor=<rgba 1 0 0.2344 1>}
  n4 = Add(A=LocalPosition() {localOrigin=InstancePreSkinning}.XYZ, B=VectorParameter() {defaultValue=<rgba -1.7698 9.4121 0 1>}.RGB)
  n5 = ScalarParameter() {defaultValue=6, parameterName=RippleScale}
  n6 = VertexInterpolator(VS=ComponentMask(None=Divide(A=@n4, B=@n5) {constB=10}) {g=true, r=true})
  n7 = Constant()
  NamedRerouteDeclaration(None=OneMinus(None=Distance(A=@n6.PS, B=@n7))) {name=CharPosGrad, nodeColor=<rgba 1 0.4453 0 1>}
  n8 = VectorParameter() {parameterName=WobbleRotationAxis, defaultValue=<rgba -0.5 1 0.5 1>}
  n9 = Multiply(
   A=MaterialFunctionCall!(FlickerInitialSpeed=ScalarParameter() {defaultValue=0.035, parameterName=WobbleSpeed}, FlickerRandom=Constant()) {materialFunction=/Game/Effects/Fort_Effects/Materials/_Functions/MF_Flicker}.Result
   B=ScalarParameter() {defaultValue=0.035, parameterName=WobbleIntensity}
  ) {constB=0.025}
  n10 = VectorParameter() {parameterName=WobblePivotOffset, defaultValue=<rgba 0 -10 0 1>}
  n11 = LocalPosition() {localOrigin=InstancePreSkinning}
  NamedRerouteDeclaration(None=Multiply(A=RotateAboutAxis(NormalizedRotationAxis=@n8.RGB, RotationAngle=@n9, PivotPoint=@n10.RGB, Position=@n11.XYZ), B=NamedRerouteUsage())) {name=LiquidWPO, nodeColor=<rgba 0.7813 0 1 1>}
  n12 = MaterialFunctionCall!(CustomUV) {materialFunction=/Game/Global/Materials/Character_Parent_2023/Anime_Parent_2023/MaterialFunctions/MFC_AnimeShader_2023_Pre}
  NamedRerouteDeclaration(None=@n12.Normal) {name=InNormal, nodeColor=<rgba 1 0.6328 0 1>}
  NamedRerouteDeclaration(None=@n12.STT) {name=InSTT, nodeColor=<rgba 0 0.4141 1 1>}
  n13 = NamedRerouteUsage()
  n14 = Multiply(A=Subtract(A=@n2.B) {constB=0.5}) {constB=0.1}
  n15 = Multiply(A=Time()) {constB=0.1}
  n16 = Constant2Vector() {g=-1, r=-0.25}
  n17 = TextureCoordinate()
  n18 = TextureSample!(UVs=Add(A=Add(A=@n1, B=@n14), B=Multiply(A=@n15, B=@n16)), Tex, "DDX(UVs)"=DDX(Value=@n17), "DDY(UVs)"=DDY(Value=@n17), Apply View MipBias) {
   mipValueMode=TMVM_Derivative
   samplerType=SAMPLERTYPE_Normal
   texture=/BRCosmetics/Athena/Effects/Particles/Skydiving/NS_Contrail_PinkSpike/Textures/T_PinkSpike_Voronoi2D_NM
  }
  n19 = NamedRerouteUsage()
  n20 = NamedRerouteUsage()
  n21 = LinearInterpolate(
   A=MaterialFunctionCall(Normal=Multiply(A=@n18.RGB, B=@n19), Flatness=ScalarParameter() {defaultValue=-0.1, parameterName=RippleIntensity}) {materialFunction=&FlattenNormal}.Result
   Alpha=Multiply(A=@n20, B=@n20)
  ) {constB=-1}
  NamedRerouteDeclaration(None=LinearInterpolate(
    A=@n13
    B=MaterialFunctionCall(BaseNormal=@n13, AdditionalNormal=@n21) {materialFunction=&BlendAngleCorrectedNormals}.Result
    Alpha=NamedRerouteUsage()
   )) {name=OutNormal, nodeColor=<rgba 0.7344 1 0 1>}
  n22 = MaterialFunctionCall(Color=NamedRerouteUsage()) {materialFunction=&SplitComponents}
  n23 = Add(A=@n22.R, B=Power(Base=NamedRerouteUsage()) {constExponent=3})
  n24 = Power(Base=NamedRerouteUsage(), Exp=ScalarParameter() {defaultValue=1, parameterName=CharPosGradPower})
  n25 = LinearInterpolate(A=@n22.G, Alpha=Multiply(A=Saturate(None=@n24), B=NamedRerouteUsage()) {desc=Make it darker where the character is}) {constB=0}
  NamedRerouteDeclaration(None=MaterialFunctionCall(R=Saturate(None=@n23) {desc=Add Liquid to Specular}, G=Saturate(None=@n25), B=@n22.B) {materialFunction=&AppendMany}.RGB) {name=OutSTT, nodeColor=<rgba 0.9453 0 1 1>}
  NamedRerouteDeclaration(None=MaterialFunctionCall(Vector or UVs=@n6.PS) {materialFunction=&VectorToRadialValue}.Radial Coordinates) {name=RadialCoords, nodeColor=<rgba 1 0 0.8203 1>}
  n26 = MaterialFunctionCall!(CustomLightVector, OverrideNormal_WS, STT=NamedRerouteUsage(), Normal_Map=NamedRerouteUsage(), DFL=@n12.DFL, FX=@n12.DFL) {materialFunction=/Game/Global/Materials/Character_Parent_2023/Anime_Parent_2023/MaterialFunctions/MFC_AnimeShader_2023_Calculations}
  SetMaterialAttributes!(
   MaterialAttributes=MaterialFunctionCall!(
    Shading_Gradient=@n26.Shading_Gradient
    Threshold_Gradient=@n26.Shading_Gradient
    Specular_Gradient=@n26.Shading_Gradient
    Sheen_Mask=@n26.Shading_Gradient
    Thickness=@n26.Shading_Gradient
    DFL_Mask=@n26.Shading_Gradient
    FX=@n26.Shading_Gradient
    Metallic_Gradient=@n26.Shading_Gradient
    RimLight_Gradient=@n26.Shading_Gradient
    NormalBuffer=@n26.Shading_Gradient
    --------
    CINE_LightSources=@n26.Shading_Gradient
    CINE_LightColor=@n26.Shading_Gradient
    ---------
    Color_Lit=@n12.Color_Lit
    Color_Shaded=@n12.Color_Lit
    SRM=@n12.Color_Lit
    Emissive=@n12.Color_Lit
    DFL_Color=@n12.Color_Lit
    Specular_Color=@n12.Color_Lit
    RimLight_Color=@n12.Color_Lit
   ) {materialFunction=/Game/Global/Materials/Character_Parent_2023/Anime_Parent_2023/MaterialFunctions/MFC_AnimeShader_2023_Post}.Result
   World Position Offset=NamedRerouteUsage()
   Emissive Color
  ) {attributeSetTypes [#1, #2]}
 ]
}
